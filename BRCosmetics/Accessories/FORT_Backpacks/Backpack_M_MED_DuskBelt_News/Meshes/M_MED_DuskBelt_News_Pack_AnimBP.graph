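# Animation Blueprint for the DuskBelt News backpack cosmetic, adding procedural spring jiggle to its base and head bones.
# The input pose gets a slight additive offset on the pack bone, then runs through four chained Anim Dynamics nodes simulating relative to that bone.
# Each spring bone gets two partial-alpha passes, one with linear and angular limits and one cone-constrained, all using upward gravity overrides and root-driven planar limits.
# The event graph has no logic.
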AnimBlueprint /BRCosmetics/Accessories/FORT_Backpacks/Backpack_M_MED_DuskBelt_News/Meshes/M_MED_DuskBelt_News_Pack_AnimBP
target_skeleton=&/M_MED_DuskBelt_News_Pack_Skeleton
graphs [
 AnimGraph {nodes [
   n1 = "Transform (Modify) Bone - Bone: pack"(ComponentPose=Local To Component(LocalPose=Input Pose()), Translation="0,1,0", Rotation="0,0,0") {
    translationMode=BMM_Additive
    translationSpace=BCS_BoneSpace
    alphaBoolBlend {blendOption=Linear}
    boneToModify {boneName=pack}
    translation {y=1}
   }
   n2 = Anim Dynamics - Bone: dyn_base_spring(ComponentPose=@n1, Alpha=0.15, ExternalForce="0,0,0") {
    angularSpringConstant=1000
    bAngularSpring=true
    bLinearSpring=true
    bUseGravityOverride=true
    linearSpringConstant=300
    simulationSpace=BoneRelative
    alphaBoolBlend {blendOption=Linear}
    boundBone {boneName=dyn_base_spring}
    componentAppliedLinearAccClamp=<2000 2000 2000>
    componentLinearAccScale=<0.6 0.3 0.3>
    componentLinearVelScale=<0.6 0.3 0.3>
    gravityOverride {z=2000}
    physicsBodyDefinitions [{
      boundBone {boneName=dyn_base_spring}
      boxExtents=<10 10 10>
      constraintSetup {
       coneAngle=10
       angularLimitsMax=<20 30 10>
       angularLimitsMin=<-60 -30 -30>
       angularTarget=<0 0 0>
       linearAxesMax=<0.4 0.4 0.5>
       linearAxesMin=<-0.4 -0.4 -1>
      }
      localJointOffset=<0 0 10>
     }]
    planarLimits [{drivingBone {boneName=root}, planeTransform {location=<5 0 0>, rotation=<rot 0 180 90>}}]
    relativeSpaceBone {boneName=pack}
   }
   n3 = Anim Dynamics - Bone: dyn_base_spring(ComponentPose=@n2, Alpha=0.1, ExternalForce="0,0,0") {
    angularSpringConstant=1000
    bAngularSpring=true
    bLinearSpring=true
    bUseGravityOverride=true
    linearSpringConstant=300
    simulationSpace=BoneRelative
    alphaBoolBlend {blendOption=Linear}
    boundBone {boneName=dyn_base_spring}
    componentAppliedLinearAccClamp=<2000 2000 2000>
    componentLinearAccScale=<0.2 0.2 0.3>
    componentLinearVelScale=<0.2 0.2 0.3>
    gravityOverride {z=2000}
    physicsBodyDefinitions [{
      boundBone {boneName=dyn_base_spring}
      boxExtents=<10 10 10>
      constraintSetup {angularConstraintType=Cone, coneAngle=10, angularLimitsMax=<20 0 0>, angularLimitsMin=<-20 0 0>, angularTarget=<0 0 0>}
      localJointOffset=<0 0 10>
     }]
    planarLimits [{drivingBone {boneName=root}, planeTransform {location=<5 0 0>, rotation=<rot 0 180 90>}}]
    relativeSpaceBone {boneName=pack}
   }
   n4 = Anim Dynamics - Bone: dyn_head_spring(ComponentPose=@n3, Alpha=0.4, ExternalForce="0,0,0") {
    angularSpringConstant=300
    bAngularSpring=true
    bLinearSpring=true
    bUseGravityOverride=true
    linearSpringConstant=250
    simulationSpace=BoneRelative
    alphaBoolBlend {blendOption=Linear}
    boundBone {boneName=dyn_head_spring}
    componentAppliedLinearAccClamp=<2000 2000 2000>
    componentLinearAccScale=<0.6 0.3 0.3>
    componentLinearVelScale=<0.6 0.3 0.3>
    gravityOverride {z=2000}
    physicsBodyDefinitions [{
      boundBone {boneName=dyn_head_spring}
      boxExtents=<5 5 5>
      constraintSetup {
       coneAngle=10
       angularLimitsMax=<20 30 10>
       angularLimitsMin=<-60 -30 -30>
       angularTarget=<0 0 0>
       linearAxesMax=<0.4 0.4 0.5>
       linearAxesMin=<-0.4 -0.4 -1>
      }
      localJointOffset=<0 0 5>
     }]
    planarLimits [{drivingBone {boneName=root}, planeTransform {location=<5 0 0>, rotation=<rot 0 180 90>}}]
    relativeSpaceBone {boneName=pack}
   }
   n5 = Anim Dynamics - Bone: dyn_head_spring(ComponentPose=@n4, Alpha=0.4, ExternalForce="0,0,0") {
    angularSpringConstant=300
    bAngularSpring=true
    bLinearSpring=true
    bUseGravityOverride=true
    linearSpringConstant=250
    simulationSpace=BoneRelative
    alphaBoolBlend {blendOption=Linear}
    boundBone {boneName=dyn_head_spring}
    componentAppliedLinearAccClamp=<2000 2000 2000>
    componentLinearAccScale=<0.2 0.2 0.3>
    componentLinearVelScale=<0.2 0.2 0.3>
    gravityOverride {z=2000}
    physicsBodyDefinitions [{
      boundBone {boneName=dyn_head_spring}
      boxExtents=<5 5 5>
      constraintSetup {angularConstraintType=Cone, coneAngle=10, angularLimitsMax=<20 0 0>, angularLimitsMin=<-20 0 0>, angularTarget=<0 0 0>}
      localJointOffset=<0 0 5>
     }]
    planarLimits [{drivingBone {boneName=root}, planeTransform {location=<5 0 0>, rotation=<rot 0 180 90>}}]
    relativeSpaceBone {boneName=pack}
   }
   Output Pose(Result=Component To Local(ComponentPose=@n5)) {layerGroup=None}
  ]}
 EventGraph {nodes [
   TryGetPawnOwner()
   Event BlueprintUpdateAnimation()
  ]}
]
variables []
